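# Season 30 story conversation (P04 Quest 04, WastelandAlternate) in which Dummy hands the player an unfinished car-parts collection job.
# A single entry tag leads to four linear speech nodes, starting with an auto-select main-questline choice and advancing via Continue.
# The first node requires the NPC to have the IronClash conversation tag, the quest active and objective obj0 incomplete.
# The chain ends after the final line, with no quest-update node.
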ConversationDatabase /BattlePassS30_Quests/Conversations/Story/P04/Quest_04/Conversation_S30_Story_SP_P04_Q04_WastelandAlternate
entry_points [{tag=Conversation.Entry.Season30.Story.P04.Q04.Dummy, destinations [#1]}]
nodes [
 ConversationEntryPointNode_0 {
  id=#1
  class=ConversationEntryPointNode
  outputs [#2]
  properties {entryTag {tagName=Conversation.Entry.Season30.Story.P04.Q04.Dummy}}
 }
 Speech_C_1 {
  id=#2
  class=Speech_C
  outputs [#3]
  properties {softTaskIcon=&T_UI_NPC_MainQuest, generalConfig {defaultMessage="Lemme guess. The ol' Legume sent you. "}}
  sub_nodes [
   ConversationChoiceNode_2 {
    class=ConversationChoiceNode
    properties {
     defaultChoiceDisplayText=Talk to Dummy
     choiceTags {gameplayTags [
       {tagName=Conversation.Choice.AutoSelectIfValid}
       {tagName=Conversation.Choice.Category.MainQuestline}
       {tagName=Athena.Quests.S30.Conversation.Story.P04.04.Dummy}
      ]}
    }
   }
   FortConversationRequirement_ControllerRequirement_1 {
    class=FortConversationRequirement_ControllerRequirement
    properties {
     nodeName=Require: NPC Has Conversation Tag
     participantId {tagName=Conversation.Participant.NPC}
     requirement {class=FortControllerRequirement_NPCHasConversationEntryTag, properties {queryTag {tagName=Conversation.Entry.IronClash}}}
    }
   }
   FortConversationRequirement_ControllerRequirement_3 {
    class=FortConversationRequirement_ControllerRequirement
    properties {
     nodeName=Require: Is Quest Active
     requirement {
      class=FortControllerRequirement_IsQuestActive
      properties {questDefinition=/BattlePassS30_Quests/Items/QuestItems/Story/P04/Quest_S30_Story_P04_Q04}
     }
    }
   }
   FortConversationRequirement_ControllerRequirement_5 {
    class=FortConversationRequirement_ControllerRequirement
    properties {
     nodeName=Require: Has Completed Quest Objective
     requirement {
      class=FortControllerRequirement_HasCompletedQuestObjective
      properties {
       bInverseCondition=true
       objectiveBackendName=Quest_S30_Story_P04_Q04_obj0
       questDefinition=&Quest_S30_Story_P04_Q04
      }
     }
    }
   }
  ]
 }
 Speech_C_28 {
  id=#3
  class=Speech_C
  outputs [#4]
  properties {generalConfig {defaultMessage="I'm workin' as fast as I can! It's just... the heat's meltin' my plastic."}}
  sub_nodes [ConversationChoiceNode_1 {class=ConversationChoiceNode, properties {defaultChoiceDisplayText=Continue}}]
 }
 Speech_C_2 {
  id=#4
  class=Speech_C
  outputs [#5]
  properties {generalConfig {defaultMessage=It's hard to collect car parts when you're turnin' into a gummy dummy.}}
  sub_nodes [ConversationChoiceNode_3 {class=ConversationChoiceNode, properties {defaultChoiceDisplayText=Continue}}]
 }
 Speech_C_0 {
  id=#5
  class=Speech_C
  outputs []
  properties {generalConfig {defaultMessage="Tell ya what - you finish the job, and I'll take a chilly snooze in my top-secret ice machine for nappin'. Deal?"}}
  sub_nodes [ConversationChoiceNode_4 {class=ConversationChoiceNode, properties {defaultChoiceDisplayText=Continue}}]
 }
]
